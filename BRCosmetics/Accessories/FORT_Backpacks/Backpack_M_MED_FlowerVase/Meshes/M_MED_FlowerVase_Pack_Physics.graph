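# Physics asset for the M_MED flower vase backpack cosmetic, providing capsule bodies and joints for its pack, string, knots and flag bones.
# The body hangs off the two-capsule pack on a free-rotating joint and carries the string, which branches to two twist-locked knots and the flag.
# The flag is a six-segment chain with tight angular limits, ending in upper, middle and lower branches.
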PhysicsAsset /BRCosmetics/Accessories/FORT_Backpacks/Backpack_M_MED_FlowerVase/Meshes/M_MED_FlowerVase_Pack_Physics
bodies {
 body_c [{
   shapeType=Capsule
   length=19.2318
   radius=3.63
   shapeName=body_c_capsule
   center=<14.5451 -2.5706 -1.5632>
   rotation=<rot -0.3451 -90 -2.488>
  }]
 flag_01 [{
   shapeType=Capsule
   length=19.2495
   radius=6.9217
   shapeName=flag_01_capsule
   center=<-1.8385 -0.3086 3.3691>
   rotation=<rot 0.2798 -90 1.6095>
  }]
 flag_02_c [{shapeType=Capsule, length=15.5681, radius=2.1836, shapeName=flag_02_c_capsule, center=<1 0 0>}]
 flag_03_c [{shapeType=Capsule, length=15.5681, radius=2.1836, shapeName=flag_03_c_capsule, center=<1 0 0>}]
 flag_04_c [{shapeType=Capsule, length=15.5681, radius=2.1836, shapeName=flag_02_c_capsule, center=<1 0 0>}]
 flag_05_c [{shapeType=Capsule, length=15.5681, radius=2.1836, shapeName=flag_02_c_capsule, center=<1 0 0>}]
 flag_06_c [{shapeType=Capsule, length=15.5681, radius=2.1836, shapeName=flag_02_c_capsule, center=<1 0 0>}]
 flag_lower_c [{shapeType=Capsule, length=9.3371, radius=2.6629, shapeName=flag_lower_c_capsule, center=<3.6842 0 -1.6506>, rotation=<rot 90 0.0002 0.0004>}]
 flag_mid_c [{
   shapeType=Capsule
   length=9.8266
   radius=2.1734
   shapeName=flag_upper_c_capsule
   center=<4.0389 0 1.5583>
   rotation=<rot 90 179.9999 -179.9999>
  }]
 flag_upper_c [{
   shapeType=Capsule
   length=9.8266
   radius=2.1734
   shapeName=flag_upper_c_capsule
   center=<4.0389 0 1.5583>
   rotation=<rot 90 179.9999 -179.9999>
  }]
 knot_lower_c [{shapeType=Capsule, length=2, radius=2, shapeName=knot_lower_c_capsule, center=<3.0246 0 0>}]
 knot_upper_c [{shapeType=Capsule, length=2, radius=2, shapeName=knot_upper_c_capsule, center=<2.3439 0 0>}]
 pack [
  {shapeType=Capsule, length=13.994, radius=10.904, shapeName=pack_capsule, center=<0 10.1279 3.0808>, rotation=<rot -0.0002 -90 -90>}
  {shapeType=Capsule, length=13.994, radius=10.904, shapeName=pack_capsule, center=<0 10.1279 -12.8992>, rotation=<rot -0.0002 -90 -90>}
 ]
 string_c [{shapeType=Capsule, length=10, radius=2, shapeName=string_c_capsule, center=<1 0 0>, rotation=<rot 0 0 90>}]
}
constraints [
 {
  bone1Name=body_c
  bone2Name=pack
  swing1LimitDegrees=45
  swing1Motion=Free
  swing2LimitDegrees=45
  swing2Motion=Free
  twistLimitDegrees=45
  twistMotion=Free
 }
 {bone1Name=string_c, bone2Name=body_c, swing1LimitDegrees=1, swing2LimitDegrees=1, twistLimitDegrees=0}
 {bone1Name=flag_03_c, bone2Name=flag_02_c, swing1LimitDegrees=2, swing2LimitDegrees=2, twistLimitDegrees=0}
 {bone1Name=flag_04_c, bone2Name=flag_03_c, swing1LimitDegrees=2, swing2LimitDegrees=2, twistLimitDegrees=0}
 {bone1Name=flag_lower_c, bone2Name=flag_06_c, swing1LimitDegrees=2, swing2LimitDegrees=2, twistLimitDegrees=0}
 {bone1Name=flag_upper_c, bone2Name=flag_06_c, swing1LimitDegrees=2, swing2LimitDegrees=2, twistLimitDegrees=0}
 {bone1Name=knot_lower_c, bone2Name=string_c, swing1LimitDegrees=15, swing2LimitDegrees=15, twistLimitDegrees=45, twistMotion=Locked}
 {bone1Name=knot_upper_c, bone2Name=string_c, swing1LimitDegrees=15, swing2LimitDegrees=15, twistLimitDegrees=45, twistMotion=Locked}
 {bone1Name=flag_02_c, bone2Name=flag_01, swing1LimitDegrees=2, swing2LimitDegrees=2, twistLimitDegrees=0}
 {bone1Name=flag_01, bone2Name=string_c, swing1LimitDegrees=2, swing2LimitDegrees=2, twistLimitDegrees=0}
 {bone1Name=flag_05_c, bone2Name=flag_04_c, swing1LimitDegrees=2, swing2LimitDegrees=2, twistLimitDegrees=0}
 {bone1Name=flag_06_c, bone2Name=flag_05_c, swing1LimitDegrees=2, swing2LimitDegrees=2, twistLimitDegrees=0}
 {bone1Name=flag_mid_c, bone2Name=flag_06_c, swing1LimitDegrees=2, swing2LimitDegrees=2, twistLimitDegrees=0}
]
