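# Animation Blueprint for the FlowerVase backpack cosmetic, animating its flag and string bones.
# The AnimGraph layers additive bone offsets through cached poses into a RigidBody simulation with overridden gravity.
# Each update computes a gravity override and eight phase-staggered sine rotators via a RotationMau macro.
# It also drives two WPO regions on the dynamic materials, raising cloth wind and turbulence while skydiving.
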AnimBlueprint /BRCosmetics/Accessories/FORT_Backpacks/Backpack_M_MED_FlowerVase/Meshes/M_MED_FlowerVase_Pack_AnimBP
target_skeleton=&/M_MED_FlowerVase_Pack_Skeleton
graphs [
 AnimGraph {nodes [
   n1 = RigidBody!(
    ComponentPose=Local To Component(LocalPose=Use cached pose 'FlagPose'())
    bAlphaBoolEnabled:bool
    OverrideWorldGravity:Vector="0,0,0"
    ExternalForce
    ComponentLinearAccScale:Vector="0.2,0.21,0.3"
    ComponentLinearVelScale:Vector="0.2,0.2,0.2"
   ) -> (Pose) {
    alphaInputType=Bool
    bClampLinearTranslationLimitToRefPose=true
    bOverrideWorldGravity=true
    overridePhysicsAsset=&/M_MED_FlowerVase_Pack_Physics
    simulationSpace=BaseBoneSpace
    alphaBoolBlend {blendInTime=1, blendOption=Linear, blendOutTime=1}
    baseBoneRef {boneName=root}
    componentLinearAccScale=<0.2 0.21 0.3>
    componentLinearVelScale=<0.2 0.2 0.2>
    simSpaceSettings {worldAlpha=0.18}
   }
   Output Pose(Result=Component To Local(ComponentPose=@n1)) {layerGroup=None}
   n2 = "Transform (Modify) Bone - Bone: string_c"!(ComponentPose=Local To Component(LocalPose=Use cached pose 'Cache Sin'()), bAlphaBoolEnabled:bool=True) -> (Pose) {
    alphaInputType=Bool
    rotationMode=BMM_Additive
    rotationSpace=BCS_BoneSpace
    alphaBoolBlend {blendInTime=0.5, blendOption=Linear, blendOutTime=0.5}
    boneToModify {boneName=string_c}
   }
   n3 = "Transform (Modify) Bone - Bone: flag_03_c"!(ComponentPose=@n2, bAlphaBoolEnabled:bool=True) -> (Pose) {
    alphaInputType=Bool
    rotationMode=BMM_Additive
    rotationSpace=BCS_BoneSpace
    alphaBoolBlend {blendInTime=0.5, blendOption=Linear, blendOutTime=0.5}
    boneToModify {boneName=flag_03_c}
   }
   n4 = "Transform (Modify) Bone - Bone: flag_04_c"!(ComponentPose=@n3, bAlphaBoolEnabled:bool=True) -> (Pose) {
    alphaInputType=Bool
    rotationMode=BMM_Additive
    rotationSpace=BCS_BoneSpace
    alphaBoolBlend {blendInTime=0.5, blendOption=Linear, blendOutTime=0.5}
    boneToModify {boneName=flag_04_c}
    rotation {pitch=-2}
   }
   n5 = "Transform (Modify) Bone - Bone: flag_05_c"!(ComponentPose=@n4, bAlphaBoolEnabled:bool=True) -> (Pose) {
    alphaInputType=Bool
    rotationMode=BMM_Additive
    rotationSpace=BCS_BoneSpace
    alphaBoolBlend {blendInTime=0.5, blendOption=Linear, blendOutTime=0.5}
    boneToModify {boneName=flag_05_c}
    rotation {pitch=-2}
   }
   n6 = "Transform (Modify) Bone - Bone: flag_06_c"!(ComponentPose=@n5, bAlphaBoolEnabled:bool=True) -> (Pose) {
    alphaInputType=Bool
    rotationMode=BMM_Additive
    rotationSpace=BCS_BoneSpace
    alphaBoolBlend {blendInTime=0.5, blendOption=Linear, blendOutTime=0.5}
    boneToModify {boneName=flag_06_c}
    rotation {pitch=-2}
   }
   n7 = "Transform (Modify) Bone - Bone: flag_07_c"!(ComponentPose=@n6, bAlphaBoolEnabled:bool=True) -> (Pose) {
    alphaInputType=Bool
    rotationMode=BMM_Additive
    rotationSpace=BCS_BoneSpace
    alphaBoolBlend {blendInTime=0.5, blendOption=Linear, blendOutTime=0.5}
    boneToModify {boneName=flag_07_c}
    rotation {pitch=-2}
   }
   n8 = "Transform (Modify) Bone - Bone: string_c"!(ComponentPose=@n7, bAlphaBoolEnabled=True, Translation:Vector="0,0,0", Rotation:Rotator="10.000,0.000,0.000") -> (Pose) {
    alphaInputType=Bool
    rotationMode=BMM_Additive
    rotationSpace=BCS_BoneSpace
    translationMode=BMM_Additive
    translationSpace=BCS_BoneSpace
    alphaBoolBlend {blendInTime=1, blendOption=Linear, blendOutTime=1}
    boneToModify {boneName=string_c}
    rotation=<rot 10 0 0>
   }
   Save cached pose 'FlagPose'(Pose=Component To Local(ComponentPose=@n8)) {cacheName=FlagPose}
   n9 = "Transform (Modify) Bone - Bone: body_c"!(ComponentPose=Local To Component(LocalPose=Input Pose()), Alpha=1, Translation="0,-3,0") -> (Pose) {
    translationMode=BMM_Additive
    translationSpace=BCS_BoneSpace
    alphaBoolBlend {blendOption=Linear}
    boneToModify {boneName=body_c}
    translation {y=-3}
   }
   n10 = "Transform (Modify) Bone - Bone: flag_01"!(ComponentPose=@n9, bAlphaBoolEnabled:bool=True, Rotation="0,0,0") -> (Pose) {
    alphaInputType=Bool
    rotationMode=BMM_Additive
    rotationSpace=BCS_BoneSpace
    translationMode=BMM_Additive
    translationSpace=BCS_BoneSpace
    alphaBoolBlend {blendInTime=0.5, blendOption=Linear, blendOutTime=0.5}
    boneToModify {boneName=flag_01}
    translation {y=-2}
   }
   n11 = "Transform (Modify) Bone - Bone: flag_02_c"!(ComponentPose=@n10, bAlphaBoolEnabled:bool=True, Rotation="0,0,0") -> (Pose) {
    alphaInputType=Bool
    rotationMode=BMM_Additive
    rotationSpace=BCS_BoneSpace
    translationSpace=BCS_BoneSpace
    alphaBoolBlend {blendInTime=0.5, blendOption=Linear, blendOutTime=0.5}
    boneToModify {boneName=flag_02_c}
   }
   n12 = "Transform (Modify) Bone - Bone: flag_03_c"!(ComponentPose=@n11, bAlphaBoolEnabled=True, Rotation:Rotator="0,0,0") -> (Pose) {
    alphaInputType=Bool
    rotationMode=BMM_Additive
    rotationSpace=BCS_BoneSpace
    translationSpace=BCS_BoneSpace
    alphaBoolBlend {blendInTime=0.5, blendOption=Linear, blendOutTime=0.5}
    boneToModify {boneName=flag_03_c}
   }
   n13 = "Transform (Modify) Bone - Bone: flag_04_c"!(ComponentPose=@n12, bAlphaBoolEnabled=True, Rotation:Rotator="0,0,0") -> (Pose) {
    alphaInputType=Bool
    rotationMode=BMM_Additive
    rotationSpace=BCS_BoneSpace
    translationSpace=BCS_BoneSpace
    alphaBoolBlend {blendInTime=0.5, blendOption=Linear, blendOutTime=0.5}
    boneToModify {boneName=flag_04_c}
   }
   n14 = "Transform (Modify) Bone - Bone: flag_05_c"!(ComponentPose=@n13, bAlphaBoolEnabled=True, Rotation:Rotator="0,0,0") -> (Pose) {
    alphaInputType=Bool
    rotationMode=BMM_Additive
    rotationSpace=BCS_BoneSpace
    translationSpace=BCS_BoneSpace
    alphaBoolBlend {blendInTime=0.5, blendOption=Linear, blendOutTime=0.5}
    boneToModify {boneName=flag_05_c}
   }
   n15 = "Transform (Modify) Bone - Bone: flag_06_c"!(ComponentPose=@n14, bAlphaBoolEnabled=True, Rotation:Rotator="0,0,0") -> (Pose) {
    alphaInputType=Bool
    rotationMode=BMM_Additive
    rotationSpace=BCS_BoneSpace
    translationSpace=BCS_BoneSpace
    alphaBoolBlend {blendInTime=0.5, blendOption=Linear, blendOutTime=0.5}
    boneToModify {boneName=flag_06_c}
   }
   n16 = "Transform (Modify) Bone - Bone: flag_07_c"!(ComponentPose=@n15, bAlphaBoolEnabled=True, Rotation:Rotator="0,0,0") -> (Pose) {
    alphaInputType=Bool
    rotationMode=BMM_Additive
    rotationSpace=BCS_BoneSpace
    translationSpace=BCS_BoneSpace
    alphaBoolBlend {blendInTime=0.5, blendOption=Linear, blendOutTime=0.5}
    boneToModify {boneName=flag_07_c}
   }
   n17 = "Transform (Modify) Bone - Bone: flag_mid_c"!(ComponentPose=@n16, bAlphaBoolEnabled:bool=True, Rotation="0,0,0") -> (Pose) {
    alphaInputType=Bool
    rotationMode=BMM_Additive
    rotationSpace=BCS_BoneSpace
    translationSpace=BCS_BoneSpace
    alphaBoolBlend {blendInTime=0.5, blendOption=Linear, blendOutTime=0.5}
    boneToModify {boneName=flag_mid_c}
   }
   n18 = "Transform (Modify) Bone - Bone: flag_lower_c"!(ComponentPose=@n17, bAlphaBoolEnabled:bool=True, Rotation="0,0,0") -> (Pose) {
    alphaInputType=Bool
    rotationMode=BMM_Additive
    rotationSpace=BCS_BoneSpace
    translationSpace=BCS_BoneSpace
    alphaBoolBlend {blendInTime=0.5, blendOption=Linear, blendOutTime=0.5}
    boneToModify {boneName=flag_lower_c}
   }
   n19 = "Transform (Modify) Bone - Bone: flag_upper_c"!(ComponentPose=@n18, bAlphaBoolEnabled:bool=True, Rotation="0,0,0") -> (Pose) {
    alphaInputType=Bool
    rotationMode=BMM_Additive
    rotationSpace=BCS_BoneSpace
    translationSpace=BCS_BoneSpace
    alphaBoolBlend {blendInTime=0.5, blendOption=Linear, blendOutTime=0.5}
    boneToModify {boneName=flag_upper_c}
   }
   Save cached pose 'Cache Sin'(Pose=Component To Local(ComponentPose=@n19)) {cacheName=Cache Sin}
   Reroute Node(InputPin:Component Space Pose Link Structure=Reroute Node()) -> (OutputPin:Component Space Pose Link Structure)
  ]}
 EventGraph {nodes [
   n1 = Event BlueprintUpdateAnimation()
   > n2 = Sequence!(execute) -> (then_0, then_1, then_2:Exec, then_3:Exec, then_4:Exec)
   Set WPO Globals Function(
    execute=@n2.then_1
    Material=Get Mat Instance Dynamic()
    Overlay Material=Get Overlay Mat Instance Dynamic()
    Skeletal Mesh=Get SkelMesh()
    "Local Noise Transform  Location"=Get Local Noise Transform Location() -> (Local Noise Transform Location:Vector)
    Local Noise Offset=Get Local Noise Offset()
   )
   > Set WPO Regional Function!(
    execute
    Material:Material Instance Dynamic Object Reference=Get Mat Instance Dynamic()
    Overlay Material:Material Instance Dynamic Object Reference=Get Overlay Mat Instance Dynamic()
    Skeletal Mesh:Skeletal Mesh Component Object Reference=Get SkelMesh()
    Velocity Transform:Name=root
    Velocity Interp Speed:double=Get Velocity Interp Speed()
    Collision Transform:Name=pack
    Collision Transform Axis:EAxis Enum
    Invert Collision Transform Axis:bool=false
    Delta Time X:double=@n1.DeltaTimeX
    Previous Velocity Transform Location:"Vector (by ref)"=Get R1 Velocity Transform Location()
    Previous Velocity Transform Velocity:"Vector (by ref)"=Get R1 Velocity Transform Velocity()
    Region:int=1
   ) -> (then)
   > Set WPO Regional Function(
    Material=Get Mat Instance Dynamic()
    Overlay Material=Get Overlay Mat Instance Dynamic()
    Skeletal Mesh=Get SkelMesh()
    Velocity Transform=root
    Velocity Interp Speed=Get Velocity Interp Speed()
    Collision Transform=pack
    Invert Collision Transform Axis=true
    Delta Time X=@n1.DeltaTimeX
    Previous Velocity Transform Location=Get R2 Velocity Transform Location()
    Previous Velocity Transform Velocity=Get R2 Velocity Transform Velocity()
    Region=2
   )
   > n3 = Branch(Condition=Get bIsSkyDiving())
   > Set WPO Pho Wind Intensity(WPO Pho Wind Intensity=0.8)
   > n4 = Set WPO Pho Turbulence(WPO Pho Turbulence:double=1.5) -> (Output_Get:double)
   SetScalarParameterValue(execute=[@n6.then, @n4.then], self=Get Mat Instance Dynamic(), ParameterName=ClothSim_WindStrength, Value=Get WPO Pho Wind Intensity())
   n5 = Get WPO Pho Turbulence() -> (WPO Pho Turbulence:double)
   > SetScalarParameterValue(self=Get Mat Instance Dynamic(), ParameterName=ClothSim_R1_TurbulenceStrength, Value=@n5)
   > SetScalarParameterValue(self=Get Mat Instance Dynamic(), ParameterName=ClothSim_R2_TurbulenceStrength, Value=@n5)
   Set WPO Pho Wind Intensity(execute=@n3.else, WPO Pho Wind Intensity=0)
   > n6 = Set WPO Pho Turbulence(WPO Pho Turbulence=0.2)
   n7 = Sequence!(execute=@n2.then_4) -> (then_0, then_1, then_2, then_3, then_4, then_5:Exec, then_6:Exec, then_7:Exec, then_8:Exec)
   n8 = Get Speed() -> (Speed:double)
   n9 = Get SpeedX() -> (SpeedX:double)
   > n10 = Rotation Mau!(Input:Exec, sin:double=@n8, Plus:double, Amplitud:double=2, Offset:double, SinX:double=@n9, PlusX:double, AmplitudX:double, OffsetX:double) -> (Output:Exec, Rotation:Rotator=~) {is=title K2Node_MacroInstance}
   > Set SinRotate1(SinRotate1:Rotator=@n10) -> (Output_Get:Rotator)
   > n11 = Rotation Mau(sin=@n8, Plus=-1, Amplitud=3, SinX=@n9) -> (Rotation=~)
   > Set SinRotate2(SinRotate2:Rotator=@n11) -> (Output_Get:Rotator)
   n12 = Rotation Mau(Input=@n7.then_1, sin=@n8, Plus=-2, Amplitud=4, SinX=@n9) -> (Rotation=~)
   > Set SinRotate3(SinRotate3:Rotator=@n12) -> (Output_Get:Rotator)
   > n13 = Rotation Mau(sin=@n8, Plus=-3, Amplitud=5, SinX=@n9) -> (Rotation=~)
   > Set SinRotate4(SinRotate4:Rotator=@n13) -> (Output_Get:Rotator)
   n14 = Rotation Mau(Input=@n7.then_2, sin=@n8, Plus=-4, Amplitud=5, SinX=@n9, PlusX=1, AmplitudX=1, OffsetX=2) -> (Rotation=~)
   > Set SinRotate5(SinRotate5:Rotator=@n14) -> (Output_Get:Rotator)
   > n15 = Rotation Mau(sin=@n8, Plus=-5, Amplitud=6, Offset=2, SinX=@n9, PlusX=2, AmplitudX=2) -> (Rotation=~)
   > Set SinRotate6(SinRotate6:Rotator=@n15) -> (Output_Get:Rotator)
   n16 = Rotation Mau(Input=@n7.then_3, sin=@n8, Plus=-6, Amplitud=7, SinX=@n9, PlusX=3) -> (Rotation=~)
   > Set SinRotate7(SinRotate7:Rotator=@n16) -> (Output_Get:Rotator)
   > n17 = Rotation Mau(sin=@n8, Plus=-7, Amplitud=8, SinX=@n9, PlusX=3) -> (Rotation=~)
   > Set SinRotate8(SinRotate8:Rotator=@n17) -> (Output_Get:Rotator)
   n18 = Gravity Override State Function DEPRECATED(
    execute=@n2.then_0
    CharacterPartAnimInstance=Self-Reference()
    GameTimeInSeconds=GetGameTimeInSeconds()
    GravityOverrideParams=Get Gravity Override Params()
   )
   > Set GravityOverride(GravityOverride=@n18)
   n19 = Get bIsMoving2D()
   > n20 = Set isIDLE(isIDLE:bool=NOT Boolean(A=@n19)) -> (Output_Get:bool)
   > Set isNotIDLE(isNotIDLE:bool=NOT Boolean(A=@n20)) -> (Output_Get:bool)
   n21 = Branch(execute=@n2.then_2, Condition=Get bIsSkyDiving())
   > Set Linear Scale WO(Linear Scale WO:Vector="0,0.2,0.1") -> (Output_Get:Vector)
   Set Linear Scale WO(execute=@n21.else, Linear Scale WO="0,0.2,0.2")
   Event OnCharacterCustomizationLoaded()
   > n22 = Set SkelMesh(SkelMesh=GetPartSkeletalMeshComponent())
   > n23 = Get Dynamic Material Instance(SkeletalMeshComponent=@n22)
   > Set Mat Instance Dynamic(Mat Instance Dynamic=@n23)
   > n24 = GetOrCreateDynamicOverlayMaterial(MeshComponent=@n22, MaterialName=None)
   > Set Overlay Mat Instance Dynamic(Overlay Mat Instance Dynamic=@n24)
   OR Boolean!(A=@n19, B=false, C:bool) -> (ReturnValue)
   Get bIsPlayingEmote()
  ]}
 NewFunction {nodes [NewFunction() {is=title K2Node_FunctionEntry}]}
 Rotation Sin {nodes [
   n1 = Rotation Sin!() -> (then, Speed:double=~, Amplitud:double=~, Rotate Y:double=~) {is=title K2Node_FunctionEntry}
   n2 = MultiplyByPi!(Value:double="Sin (Radians)"(A=@n1)) -> (ReturnValue:double)
   > Return Node!(execute, Rotation:Rotator=MakeRotator(Pitch=@n1.Rotate Y, Yaw=float * float(A=@n2, B=@n1.Amplitud) {is=title K2Node_PromotableOperator}))
  ]}
 RotationMau {nodes [
   n1 = Inputs!() -> (
    Input:Exec
    sin:double=~
    Plus:double=~
    Amplitud:double=~
    Offset:double=~
    SinX:double=~
    PlusX:double=~
    AmplitudX:double=~
    OffsetX:double=~
   )
   n2 = GetRealTimeSeconds()
   n3 = float + float(A=float * float(A=@n2, B=@n1.SinX), B=@n1.PlusX)
   n4 = float * float(A=MultiplyByPi(Value="Sin (Radians)"(A=@n3)), B=@n1.AmplitudX)
   n5 = float + float(A=float * float(A=@n2, B=@n1), B=@n1.Plus)
   n6 = float * float(A=MultiplyByPi(Value="Sin (Radians)"(A=@n5)), B=@n1.Amplitud)
   > Outputs!(Output:Exec, Rotation:Rotator=MakeRotator(Pitch=float + float(A=@n4, B=@n1.OffsetX), Yaw=float + float(A=@n6, B=@n1.Offset)))
   Get Speed()
  ]}
]
variables [
 Gravity Override Params {
  type=GravityOverrideParamsStruct
  default {skydive_FloatingDown {gravityOverride=<0 -500 200>, windAmplitude=<0 200 100>, windFrequency=<1.5 20 10>, windOffset=<0 100 100>}}
 }
 GravityOverride {type=Vector}
 isIDLE {type=boolean}
 isNotIDLE {type=boolean}
 Linear Scale WO {type=Vector}
 Local Noise Offset {type=Vector}
 Local Noise Transform Location {type=Vector}
 Mat Instance Dynamic {type=MaterialInstanceDynamic}
 Overlay Mat Instance Dynamic {type=MaterialInstanceDynamic}
 R1 Velocity Transform Location {type=Vector}
 R1 Velocity Transform Velocity {type=Vector}
 R2 Velocity Transform Location {type=Vector}
 R2 Velocity Transform Velocity {type=Vector}
 SinRotate1 {type=Rotator}
 SinRotate2 {type=Rotator}
 SinRotate3 {type=Rotator}
 SinRotate4 {type=Rotator}
 SinRotate5 {type=Rotator}
 SinRotate6 {type=Rotator}
 SinRotate7 {type=Rotator}
 SinRotate8 {type=Rotator}
 SkelMesh {type=SkeletalMeshComponent}
 Speed {type=number, default=3.2}
 SpeedX {type=number, default=1.5}
 Velocity Interp Speed {type=number}
 WPO Pho Turbulence {type=number}
 WPO Pho Wind Intensity {type=number}
]
